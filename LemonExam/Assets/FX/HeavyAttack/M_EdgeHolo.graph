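# M_EdgeHolo is a translucent, unlit, two-sided Niagara material for a holographic edge in the HeavyAttack effects.
# Color hue-shifts a base color over time and along U, then multiplies by tint and particle color into an emissive adjustment function.
# Procedural edge masks build a tapered shape with faded tips, sides and center.
# View angle, vertex green, particle alpha and a fourth-power depth fade further attenuate that mask.
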Material /LemonExam/Assets/FX/HeavyAttack/M_EdgeHolo
graph {
 properties {
  bAutomaticallySetUsageInEditor=false
  bUsedWithNiagaraMeshParticles=true
  bUsedWithNiagaraRibbons=true
  bUsedWithNiagaraSprites=true
  blendMode=BLEND_Translucent
  shadingModel=MSM_Unlit
  twoSided=true
  usedShadingModels=Unlit
 }
 nodes [
  n1 = ParticleColor()
  n2 = Time()
  n3 = ScalarParameter() {parameterName=OffsetSpeed}
  n4 = VectorParameter() {parameterName=BaseC, defaultValue=<rgba 0.0678 0.4087 1 8>}
  n5 = MaterialFunctionCall(Float2=TextureCoordinate()) {materialFunction=&BreakOutFloat2Components}
  n6 = MaterialFunctionCall(Hue Shift Percentage=Add(A=Multiply(A=@n2, B=@n3), B=Multiply(A=@n4.A, B=@n5.R)), Texture=@n4.RGB) {materialFunction=&HueShift}
  n7 = VectorParameter() {parameterName=M, defaultValue=<rgba 1.4574 1.7027 2 1>}
  MaterialFunctionCall(Emissive=Multiply(A=@n1.RGB, B=Multiply(A=@n6.Result, B=@n7.RGB))) {materialFunction=&MF_FX_EmissiveAdjustments}
  n8 = VectorParameter() {
   group=TipTailFade
   parameterName=TipTailFade_EdgeMask_Control
   channelNames {a="", b=Power, g=Multiply, r=Offset}
   defaultValue=<rgba 0 3.8442 2 1>
  }
  n9 = ScalarParameter() {defaultValue=1, parameterName=T}
  n10 = VectorParameter() {
   group=ScaleB
   parameterName=ScaleB_EdgeMask_Control
   channelNames {a="", b=Power, g=Multiply, r=Offset}
   defaultValue=<rgba 0 3.549 0.6116 1>
  }
  n11 = VectorParameter() {group=SideFade, parameterName=SideFade_EdgeMask_Control, channelNames {a="", b=Power, g=Multiply, r=Offset}, defaultValue=<rgba 0 2 2 1>}
  n12 = MaterialFunctionCall!(
   UorV=MaterialFunctionCall(
    UVs=@n5.G
    Texture Scale=MaterialFunctionCall(UorV=@n5.R, EdgeControl=@n10) {materialFunction=/Game/Effects/Niagara/MasterMaterials/Functions/Common/FXCommon_EdgeMask_Simple}.Mask
   ) {materialFunction=&ScaleUVsByCenter}.UVs
   EdgeControl=@n11
  ) {materialFunction=&FXCommon_EdgeMask_Simple}
  n13 = VectorParameter() {
   group=CenterFade
   parameterName=CenterFade_EdgeMask_Control
   channelNames {a="", b=Power, g=Multiply, r=Offset}
   defaultValue=<rgba 0 2.5138 0.5 1>
  }
  n14 = OneMinus(None=MaterialFunctionCall(UorV=@n5.R, EdgeControl=@n13) {materialFunction=&FXCommon_EdgeMask_Simple}.Mask)
  n15 = LinearInterpolate(
   A=Multiply(
    A=MaterialFunctionCall("FadeSensitivity (optional)"=@n9) {materialFunction=&FadeBasedOnViewAngle}.Result
    B=Multiply(A=@n12.Mask, B=@n14)
   )
   Alpha=VertexColor().G
  ) {constB=0}
  n16 = DepthFade(FadeDistance=ScalarParameter() {defaultValue=300, parameterName=DF})
  n17 = Multiply(A=@n16, B=@n16)
  Multiply(
   A=Multiply(
    A=MaterialFunctionCall(UorV=@n5.R, EdgeControl=@n8) {materialFunction=&FXCommon_EdgeMask_Simple}.Mask
    B=Multiply(A=@n1.A, B=@n15)
   )
   B=Multiply(A=@n17, B=@n17)
  )
  VectorParameter() {parameterName=Sca, defaultValue=<rgba 0.4048 1 0 1>}
 ]
}
