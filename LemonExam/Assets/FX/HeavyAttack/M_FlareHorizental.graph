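# Additive, unlit, two-sided Niagara material for a horizontal flare in the HeavyAttack effects.
# Color is hue-shifted along U by an amount scaled by its alpha, tinted by particle color, then passed through MF_FX_EmissiveAdjustments.
# A soft SphereMask centered in UV space, shaped by MF_FX_Refine and Shape_Refine, is multiplied by particle alpha.
# That mask is multiplied by a cubed DepthFade to soften geometry intersections with a steep falloff.
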Material /LemonExam/Assets/FX/HeavyAttack/M_FlareHorizental
graph {
 properties {
  bAutomaticallySetUsageInEditor=false
  bUsedWithNiagaraMeshParticles=true
  bUsedWithNiagaraRibbons=true
  bUsedWithNiagaraSprites=true
  blendMode=BLEND_Additive
  shadingModel=MSM_Unlit
  twoSided=true
  usedShadingModels=Unlit
 }
 nodes [
  n1 = VectorParameter() {parameterName=Color, defaultValue=<rgba 1 0.3135 0.3135 3>}
  n2 = Multiply(
   A=MaterialFunctionCall(Float2=TextureCoordinate()) {materialFunction=&BreakOutFloat2Components}.R
   B=@n1.A
  )
  n3 = ParticleColor()
  MaterialFunctionCall(Emissive=Multiply(
    A=MaterialFunctionCall(Hue Shift Percentage=@n2, Texture=@n1.RGB) {materialFunction=&HueShift}.Result
    B=@n3.RGB
   )) {materialFunction=&MF_FX_EmissiveAdjustments}
  Multiply()
  n4 = SphereMask(A=TextureCoordinate(), B=Constant() {r=0.5}) {attenuationRadius=0.5, hardnessPercent=0}
  n5 = VectorParameter() {parameterName=Shape_Refine, defaultValue=<rgba 1.25 0.5 1 0>}
  n6 = DepthFade(FadeDistance=ScalarParameter() {defaultValue=400, parameterName=DF})
  Multiply(
   A=Multiply(
    A=@n3.A
    B=MaterialFunctionCall(Texture=@n4, RGB=@n5.RGB, A=@n5.RGB) {materialFunction=&MF_FX_Refine}.Result
   )
   B=Multiply(A=Multiply(A=@n6, B=@n6), B=@n6)
  )
 ]
}
