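# Translucent, unlit, two-sided Niagara material for sprites, ribbons and mesh particles, apparently a segmented heavy-attack slash test.
# A dynamic parameter packs six decimal digits, extracted by divide and modulo, to darken individual sixths along U.
# One branch builds hard Ceil-based segment masks; another builds adjustable soft falloffs using distance and Power.
# EdgeMask functions shape segment and UV borders, alongside an HDR Color parameter.
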Material /LemonExam/Assets/FX/HeavyAttack/M_TestSlashPart
graph {
 properties {
  bAutomaticallySetUsageInEditor=false
  bUsedWithNiagaraMeshParticles=true
  bUsedWithNiagaraRibbons=true
  bUsedWithNiagaraSprites=true
  blendMode=BLEND_Translucent
  shadingModel=MSM_Unlit
  twoSided=true
  usedShadingModels=Unlit
 }
 nodes [
  n1 = MaterialFunctionCall(Float2=TextureCoordinate()) {materialFunction=&BreakOutFloat2Components}
  n2 = Multiply(A=@n1.R) {constB=6}
  n3 = Add(A=@n2) {constB=0.5}
  n4 = Frac(None=@n3)
  n5 = VectorParameter() {group=Segment, parameterName=Segment2_EdgeMask_Control, channelNames {a="", b=Power, g=Multiply, r=Offset}, defaultValue=<rgba 0 2.2285 5 1>}
  Add(A=Multiply(A=MaterialFunctionCall!(UorV=@n4, EdgeControl=@n5) {materialFunction=/Game/Effects/Niagara/MasterMaterials/Functions/Common/FXCommon_EdgeMask_Simple}.Mask) {constB=0.5})
  Divide(A=Ceil(None=@n3)) {constB=6}
  n6 = Frac(None=@n2)
  n7 = VectorParameter() {group=Segment, parameterName=Segment_EdgeMask_Control, channelNames {a="", b=Power, g=Multiply, r=Offset}, defaultValue=<rgba 0 2 2 1>}
  n8 = Divide(A=Ceil(None=@n2)) {constB=6}
  n9 = DynamicParameter() {defaultValue=<rgba 111111 1 1 1>, paramNames [Mask,Param2,Param3,Param4]}
  n10 = Constant() {r=10}
  n11 = Saturate(None=Modulo!(A=Divide(A=@n9.Mask) {constB=100000}, B=@n10))
  n12 = Constant() {r=6}
  n13 = Multiply(A=@n11, B=Divide(B=@n12)) {constB=0.1}
  n14 = Subtract(A=Abs(None=Subtract(A=@n8, B=@n13))) {constB=0.01}
  n15 = Saturate(None=Modulo(A=Divide(A=@n9.Mask) {constB=10000}, B=@n10))
  n16 = Multiply(A=@n15, B=Divide(B=@n12) {constA=2}) {constB=0.2}
  n17 = Subtract(A=Abs(None=Subtract(A=@n8, B=@n16))) {constB=0.01}
  n18 = Saturate(None=Modulo(A=Divide(A=@n9.Mask) {constB=1000}, B=@n10))
  n19 = Multiply(A=@n18, B=Divide(B=@n12) {constA=3}) {constB=0.2}
  n20 = Abs(None=Subtract(A=@n8, B=@n19))
  n21 = Multiply(A=Multiply(A=Ceil(None=@n14), B=Ceil(None=@n17)), B=Ceil(None=Subtract(A=@n20) {constB=0.01}))
  n22 = Saturate(None=Modulo(A=Divide(A=@n9.Mask) {constB=100}, B=@n10))
  n23 = Multiply(A=Divide(B=@n12) {constA=4}, B=@n22) {constB=0.1}
  n24 = Subtract(A=@n8, B=@n23)
  n25 = Ceil(None=Subtract(A=Abs(None=@n24)) {constB=0.01})
  n26 = Saturate(None=Modulo(A=Divide(A=@n9.Mask) {constB=10}, B=@n10))
  n27 = Multiply(A=Divide(B=@n12) {constA=5}, B=@n26) {constB=0.2}
  n28 = Subtract(A=Abs(None=Subtract(A=@n8, B=@n27))) {constB=0.01}
  n29 = Saturate(None=Modulo(A=Divide(A=@n9.Mask) {constB=1}, B=@n10))
  n30 = Multiply(A=@n29, B=Divide(B=@n12) {constA=6}) {constB=0.2}
  n31 = Abs(None=Subtract(A=@n8, B=@n30))
  n32 = Multiply(A=Multiply(A=Multiply(A=@n21, B=@n25), B=Ceil(None=@n28)), B=Ceil(None=Subtract(A=@n31) {constB=0.01}))
  Multiply(A=Multiply(
    A=MaterialFunctionCall(UorV=@n6, EdgeControl=@n7) {materialFunction=&FXCommon_EdgeMask_Simple}.Mask
    B=Saturate(None=@n32)
   )) {constB=0.5}
  n33 = MaterialFunctionCall(Float2=TextureCoordinate()) {materialFunction=&BreakOutFloat2Components}
  n34 = VectorParameter() {group=EDGE, parameterName=EDGE_EdgeMask_Control, channelNames {a="", b=Power, g=Multiply, r=Offset}, defaultValue=<rgba 0 18.182 3 1>}
  n35 = Add(A=@n1.R, B=Divide(A=Divide(B=@n12)))
  n36 = Abs(None=Subtract(A=@n35, B=@n13))
  n37 = ScalarParameter() {group=SegMask, parameterName=SegMask_GrayAdjust_Add}
  n38 = ScalarParameter() {defaultValue=8.578, group=SegMask, parameterName=SegMask_GrayAdjust_Mul}
  n39 = Multiply(A=Add(A=LinearInterpolate(B=@n36, Alpha=@n11) {constA=1}, B=@n37), B=@n38)
  n40 = ScalarParameter() {defaultValue=8.3482, group=SegMask, parameterName=SegMask_GrayAdjust_Power}
  n41 = Abs(None=Subtract(A=@n35, B=@n16))
  n42 = Multiply(A=Add(A=LinearInterpolate(B=@n41, Alpha=@n15) {constA=1}, B=@n37), B=@n38)
  n43 = Multiply(A=Power(Base=Saturate(None=@n39), Exp=@n40), B=Power(Base=Saturate(None=@n42), Exp=@n40))
  n44 = Subtract(A=@n35, B=@n19)
  n45 = Add(A=LinearInterpolate(B=Abs(None=@n44), Alpha=@n18) {constA=1}, B=@n37)
  n46 = Power(Base=Saturate(None=Multiply(A=@n45, B=@n38)), Exp=@n40)
  n47 = LinearInterpolate(B=Abs(None=Subtract(A=@n35, B=@n23)), Alpha=@n22) {constA=1}
  n48 = Saturate(None=Multiply(A=Add(A=@n47, B=@n37), B=@n38))
  n49 = Abs(None=Subtract(A=@n35, B=@n27))
  n50 = Multiply(A=Add(A=LinearInterpolate(B=@n49, Alpha=@n26) {constA=1}, B=@n37), B=@n38)
  n51 = Multiply(A=Multiply(A=Multiply(A=@n43, B=@n46), B=Power(Base=@n48, Exp=@n40)), B=Power(Base=Saturate(None=@n50), Exp=@n40))
  n52 = Subtract(A=@n35, B=@n30)
  n53 = Add(A=LinearInterpolate(B=Abs(None=@n52), Alpha=@n29) {constA=1}, B=@n37)
  n54 = Power(Base=Saturate(None=Multiply(A=@n53, B=@n38)), Exp=@n40)
  n55 = VectorParameter() {group=EDGE, parameterName=EDGE2_EdgeMask_Control, channelNames {a="", b=Power, g=Multiply, r=Offset}, defaultValue=<rgba 0 1 2 1>}
  Multiply(
   A=Multiply(
    A=MaterialFunctionCall(UorV=@n33.R, EdgeControl=@n34) {materialFunction=&FXCommon_EdgeMask_Simple}.Mask
    B=Multiply(A=@n51, B=@n54)
   )
   B=MaterialFunctionCall(UorV=@n33.G, EdgeControl=@n55) {materialFunction=&FXCommon_EdgeMask_Simple}.Mask
  )
  VectorParameter() {parameterName=Color, defaultValue=<rgba 3.4334 3.6896 5 1>}
 ]
}
