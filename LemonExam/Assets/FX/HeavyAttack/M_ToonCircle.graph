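# Translucent unlit, two-sided Niagara material for the heavy attack FX, drawing a toon-style circle that dissolves as a ring.
# The ring shape subtracts one UV sphere mask from another, with the inner radius grown by a dynamic Dissolve parameter.
# A refined mask blends a tinted, saturation-adjusted rainbow texture with particle color for emissive.
# The same ring, sharpened through power and refine functions, multiplies particle alpha for opacity.
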Material /LemonExam/Assets/FX/HeavyAttack/M_ToonCircle
graph {
 properties {
  bAutomaticallySetUsageInEditor=false
  bUsedWithNiagaraMeshParticles=true
  bUsedWithNiagaraRibbons=true
  bUsedWithNiagaraSprites=true
  blendMode=BLEND_Translucent
  shadingModel=MSM_Unlit
  twoSided=true
  usedShadingModels=Unlit
 }
 nodes [
  n1 = TextureObjectParameter() {group=Rainbow, parameterName=Rainbow_Texture, sortPriority=0, texture=&T_Tile_RainbowColor}
  n2 = VectorParameter() {group=Rainbow, parameterName=Rainbow_UVControl, channelNames {a=VPan, b=UPan, g=VScale, r=UScale}, defaultValue=<rgba 1 1 0 0>}
  n3 = Desaturation(
   None=Normalize(VectorInput=MaterialFunctionCall(Texture=@n1, UVControl_RGB=@n2, UVControl_A=@n2) {materialFunction=&FXCommon_TextureSample}.RGB)
   Fraction=ScalarParameter() {defaultValue=-0.9536, parameterName=DES}
  )
  n4 = VectorParameter() {parameterName=C2, defaultValue=<rgba 0.4696 0.4696 2 1>}
  n5 = ParticleColor()
  n6 = TextureCoordinate()
  n7 = ScalarParameter() {defaultValue=0.8832, parameterName=Soft}
  n8 = VectorParameter() {parameterName=Center2, defaultValue=<rgba 0.4892 0.5116 0.3436 1>}
  n9 = DynamicParameter() {defaultValue=<rgba 0 1 1 1>, paramNames [DIssolve,Param2,Param3,Param4]}
  n10 = Subtract(
   A=SphereMask(A=@n6, B=Constant() {r=0.5}, Hardness=@n7) {attenuationRadius=0.4}
   B=SphereMask(A=@n6, B=AppendVector(A=@n8.R, B=@n8.R), Radius=Add(A=@n9.DIssolve, B=@n8.B), Hardness=@n7) {attenuationRadius=0.3}
  )
  n11 = VectorParameter() {parameterName=ColorRefine_Refine, defaultValue=<rgba 2830.1719 7.0327 0 0.5>}
  n12 = MaterialFunctionCall(Texture=Saturate(None=@n10), RGB=@n11.RGB, A=@n11.RGB) {materialFunction=&MF_FX_Refine}
  MaterialFunctionCall(Emissive=LinearInterpolate(A=Multiply(A=@n3, B=@n4.RGB), B=@n5.RGB, Alpha=Saturate(None=@n12.Result))) {materialFunction=&MF_FX_EmissiveAdjustments}
  n13 = Power(Base=Saturate(None=@n10), Exp=ScalarParameter() {defaultValue=2, parameterName=BP})
  n14 = VectorParameter() {parameterName=OpacityRefine_Refine, defaultValue=<rgba 33.1242 235.3683 0.8208 0.5>}
  Multiply(
   A=Saturate(None=MaterialFunctionCall(Texture=@n13, RGB=@n14.RGB, A=@n14.RGB) {materialFunction=&MF_FX_Refine}.Result)
   B=@n5.A
  )
 ]
}
