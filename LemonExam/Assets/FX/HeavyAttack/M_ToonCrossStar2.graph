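# Unlit translucent Niagara material that draws a cross-star effect with a tinted rainbow outer color.
# Center and edge masks each combine two CubeGradient samples with swapped center scales, using max for the center and screen blending for the edge.
# A power curve skews U, and dynamic parameter Tight adjusts arm scale.
# The center mask blends rainbow toward particle color for emissive; center plus edge, times particle alpha, forms opacity.
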Material /LemonExam/Assets/FX/HeavyAttack/M_ToonCrossStar2
graph {
 properties {
  bAutomaticallySetUsageInEditor=false
  bUsedWithNiagaraMeshParticles=true
  bUsedWithNiagaraRibbons=true
  bUsedWithNiagaraSprites=true
  blendMode=BLEND_Translucent
  shadingModel=MSM_Unlit
  usedShadingModels=Unlit
 }
 nodes [
  n1 = TextureObjectParameter() {group=Rainbow, parameterName=Rainbow_Texture, sortPriority=0, texture=&T_Tile_RainbowColor}
  n2 = VectorParameter() {group=Rainbow, parameterName=Rainbow_UVControl, channelNames {a=VPan, b=UPan, g=VScale, r=UScale}, defaultValue=<rgba 1 1 0 0>}
  n3 = Desaturation(
   None=Normalize(VectorInput=MaterialFunctionCall(Texture=@n1, UVControl_RGB=@n2, UVControl_A=@n2) {materialFunction=&FXCommon_TextureSample}.RGB)
   Fraction=ScalarParameter() {defaultValue=-2.4645, parameterName=DES}
  )
  n4 = VectorParameter() {parameterName=C2, defaultValue=<rgba 0.4696 0.4696 2 1>}
  n5 = ParticleColor()
  n6 = MaterialFunctionCall(Float2=TextureCoordinate()) {materialFunction=&BreakOutFloat2Components}
  n7 = MaterialFunctionCall(X=Power(Base=@n6.R, Exp=ScalarParameter() {defaultValue=0.6942, parameterName=Testtt}), Y=@n6.G) {materialFunction=&MakeFloat2}
  n8 = VectorParameter() {parameterName=ScaleByCenterB, defaultValue=<rgba 0.8075 0.0184 0 1>}
  n9 = Multiply(A=DynamicParameter() {defaultValue=<rgba 0 1 1 1>, paramNames [Tight,Param2,Param3,Param4]}.Tight) {constB=0.01}
  n10 = AppendVector(A=@n8.R, B=Add(A=@n8.G, B=@n9))
  n11 = Rotator(
   Coordinate=MaterialFunctionCall(UVs=@n7.Result, Texture Scale=@n10) {materialFunction=&ScaleUVsByCenter}.UVs
   Time=MaterialFunctionCall() {materialFunction=&Pi}.Result
  )
  n12 = Constant2Vector() {g=1, r=1}
  n13 = Multiply(
   A=TextureSample(UVs=@n11) {samplerType=SAMPLERTYPE_LinearGrayscale, texture=&/CubeGradient}.R
   B=MaterialFunctionCall!(TexCoord=@n11, ClampChannel=@n12) {materialFunction=/Game/Effects/Fort_Effects/MaterialFunctions/MF_FX_TilingMethod}.Result
  )
  n14 = Rotator(
   Coordinate=MaterialFunctionCall(
    UVs=@n7.Result
    Texture Scale=MaterialFunctionCall(XY=@n10) {materialFunction=&Swizzle}.YX
   ) {materialFunction=&ScaleUVsByCenter}.UVs
   Time=MaterialFunctionCall() {materialFunction=&Pi}.Result
  )
  n15 = Multiply(
   A=TextureSample(UVs=@n14) {samplerType=SAMPLERTYPE_LinearGrayscale, texture=&CubeGradient}.R
   B=MaterialFunctionCall(TexCoord=@n14, ClampChannel=@n12) {materialFunction=&MF_FX_TilingMethod}.Result
  )
  n16 = Multiply(
   A=Add(A=Max(A=@n13, B=@n15), B=ScalarParameter() {group=Center, parameterName=Center_GrayAdjust_Add})
   B=ScalarParameter() {defaultValue=1.0525, group=Center, parameterName=Center_GrayAdjust_Mul}
  )
  n17 = Saturate(None=Power(Base=Saturate(None=@n16), Exp=ScalarParameter() {defaultValue=1.7897, group=Center, parameterName=Center_GrayAdjust_Power}))
  MaterialFunctionCall(Emissive=LinearInterpolate(A=Multiply(A=@n3, B=@n4.RGB), B=@n5.RGB, Alpha=@n17)) {materialFunction=&MF_FX_EmissiveAdjustments}
  n18 = VectorParameter() {parameterName=ScaleByCenterA, defaultValue=<rgba 0.7164 0.1272 0 1>}
  n19 = AppendVector(A=@n18.R, B=Add(A=@n18.G, B=@n9))
  n20 = Rotator(
   Coordinate=MaterialFunctionCall(UVs=@n7.Result, Texture Scale=@n19) {materialFunction=&ScaleUVsByCenter}.UVs
   Time=MaterialFunctionCall() {materialFunction=&Pi}.Result
  )
  n21 = Constant2Vector() {g=1, r=1}
  n22 = OneMinus(None=Multiply(
    A=TextureSample(UVs=@n20) {samplerType=SAMPLERTYPE_LinearGrayscale, texture=&CubeGradient}.R
    B=MaterialFunctionCall(TexCoord=@n20, ClampChannel=@n21) {materialFunction=&MF_FX_TilingMethod}.Result
   ))
  n23 = Rotator(
   Coordinate=MaterialFunctionCall(
    UVs=@n7.Result
    Texture Scale=MaterialFunctionCall(XY=@n19) {materialFunction=&Swizzle}.YX
   ) {materialFunction=&ScaleUVsByCenter}.UVs
   Time=MaterialFunctionCall() {materialFunction=&Pi}.Result
  )
  n24 = OneMinus(None=Multiply(
    A=TextureSample(UVs=@n23) {samplerType=SAMPLERTYPE_LinearGrayscale, texture=&CubeGradient}.R
    B=MaterialFunctionCall(TexCoord=@n23, ClampChannel=@n21) {materialFunction=&MF_FX_TilingMethod}.Result
   ))
  n25 = Add(A=OneMinus(None=Multiply(A=@n22, B=@n24)), B=ScalarParameter() {group=Edge, parameterName=Edge_GrayAdjust_Add})
  n26 = Saturate(None=Power(Base=Saturate(None=@n25), Exp=ScalarParameter() {defaultValue=2, group=Edge, parameterName=Edge_GrayAdjust_Power}))
  n27 = ScalarParameter() {defaultValue=1, group=Edge, parameterName=Edge_GrayAdjust_Mul}
  Multiply(A=@n5.A, B=Add(A=@n17, B=Multiply(A=@n26, B=@n27)))
 ]
}
