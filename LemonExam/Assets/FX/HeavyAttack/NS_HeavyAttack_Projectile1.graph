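# Niagara system for a heavy attack projectile effect, in which only the mesh emitter Minimal003 is enabled.
# Minimal003 bursts a rotated, offset mesh, runs the User.ActiveArray bool array through an ArrayToInt module, and feeds Particles.Mask and a rising curve into dynamic material parameters.
# The other emitters are disabled: CombineMesh, which merges six mesh layers using spawn-group switches, plus Flow meshes, streak and large sprites, and an empty ParticleLight.
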NiagaraSystem /LemonExam/Assets/FX/HeavyAttack/NS_HeavyAttack_Projectile1
dependencies {
 usedDataInterfaces [&NiagaraDataInterfaceColorCurve, &NiagaraDataInterfaceCurve, &NiagaraDataInterfaceVectorCurve]
 usedDynamicInputs []
 usedModules [
  &self:SystemSpawnScript.SystemScriptSource.SystemScriptGraph.NiagaraNodeAssignment_0.SetVariables_#1
  &self:SystemSpawnScript.SystemScriptSource.SystemScriptGraph.NiagaraNodeAssignment_1.SetVariables_#2
  &self:SystemSpawnScript.SystemScriptSource.SystemScriptGraph.NiagaraNodeAssignment_2.SetVariables_#3
  &SystemState
  &self:SystemSpawnScript.SystemScriptSource.SystemScriptGraph.NiagaraNodeAssignment_3.SetVariables_#4
 ]
}
emitters {
 CombineMesh {
  bEnabled=false
  emitterUpdateScript [
   &EmitterState(Life Cycle Mode=::NewEnumerator1, Loop Behavior=::NewEnumerator1, Loop Duration=1, UseLoopDelay=false)
   &SpawnBurst_Instantaneous(Use Spawn Probability=false, UseLoopCountLimit=false)
   &SpawnBurst_Instantaneous(Spawn Group=1, Use Spawn Probability=false, UseLoopCountLimit=false) {module=SpawnBurst_Instantaneous001}
   &SpawnBurst_Instantaneous(Spawn Group=2, Use Spawn Probability=false, UseLoopCountLimit=false) {module=SpawnBurst_Instantaneous002}
   &SpawnBurst_Instantaneous(Spawn Group=3, Use Spawn Probability=false, UseLoopCountLimit=false) {module=SpawnBurst_Instantaneous003}
   &SpawnBurst_Instantaneous(Spawn Group=4, Use Spawn Probability=false, UseLoopCountLimit=false) {module=SpawnBurst_Instantaneous004}
   &SpawnBurst_Instantaneous(Spawn Group=5, Use Spawn Probability=false, UseLoopCountLimit=false) {module=SpawnBurst_Instantaneous005}
  ]
  particleSpawnScript [
   &InitializeParticle(
    Color=&MergeHelper_ColorSwitch(
     0=<rgba 7 7 7 1>
     1=<rgba 10 10 10 1>
     2=<rgba 0.8 0.8 0.8 2>
     3=<rgba 1 1 1 1>
     4=<rgba 1 1 1 1>
     5=<rgba 0.0509 0.0446 0.1357 4>
     6=<rgba 0 0 0 0>
     7=<rgba 0 0 0 0>
     8=<rgba 0 0 0 0>
     9=<rgba 0 0 0 0>
     Index=&SpawnGroup()
     Num=::NewEnumerator8
    )
    Lifetime=&Multiply_Float(A=1, B=1)
    Mesh Scale=&MergeHelper_VectorSwitch(
     0=&Multiply_VectorByFloat(Vector=<1 1 0.1>, Float=1.95)
     1=&Multiply_VectorByFloat(Vector=<1 0.92 0.5>, Float=1.761)
     2=&Multiply_VectorByFloat(Vector=<1 1 0.4>, Float=1.9)
     3=&Multiply_VectorByFloat(Vector=<1.1383 1 0.2>, Float=1.9683)
     4=&Multiply_VectorByFloat(Vector=<1 0.92 0.2>, Float=2.385)
     5=&Multiply_VectorByFloat(Vector=<1 1.2126 0.5415>, Float=1.6307)
     6=<0 0 0>
     7=<0 0 0>
     8=<0 0 0>
     9=<0 0 0>
     Index=&SpawnGroup()
     Num=::NewEnumerator8
    )
    Mesh Scale Mode=::NewEnumerator4
    Position Offset=&MergeHelper_VectorSwitch(
     0=&AddVectorToPosition(Position=<14.8083 0 0>, Vector=$System.FXPosition)
     1=&AddVectorToPosition(Position=<0 0 -3.028>, Vector=$System.FXPosition)
     2=&AddVectorToPosition(Position=<0 0 0>, Vector=$System.FXPosition)
     3=&AddVectorToPosition(Position=<-39.2158 0 0>, Vector=$System.FXPosition)
     4=&AddVectorToPosition(Position=<8 0 0>, Vector=$System.FXPosition)
     5=&AddVectorToPosition(Position=<-60.5189 0 0>, Vector=$System.FXPosition)
     6=<0 0 0>
     7=<0 0 0>
     8=<0 0 0>
     9=<0 0 0>
     Index=&SpawnGroup()
     Num=::NewEnumerator8
    )
    UsePositionOffset=true
    Write Material Random=true
   )
   &InitialMeshOrientation(Mesh Orientation Mode=::NewEnumerator7, Rotation=<0 0 0.5>, Use Rotation Vector=true) {module=InitialMeshOrientation001}
   &InitialMeshOrientation(Mesh Orientation Mode=::NewEnumerator7, Use Rotation Vector=true)
   /SetVariables(Particles.VisibilityTag=&SpawnGroup()) {isSetParameters=true}
  ]
  particleUpdateScript [
   &ParticleState()
   &ScaleMeshSize(Scale Factor=$System.SizeScale)
   &ScaleColor(
    Scale Alpha=&MergeHelper_FloatSwitch(
     0=&Multiply_Float(
      A=10
      B=&FloatFromCurve(FloatCurve=DataInterface() {propertyValues {Curve {keys [<at 0 1>,<at 0.7 1>,<at 1 0>]}}}, Scale Curve=2)
     )
     1=&Multiply_Float(
      A=0.7
      B=&FloatFromCurve(FloatCurve=DataInterface() {propertyValues {Curve {keys [
           <at 0 0 RCIM_Cubic>
           {arriveTangent=1.4583, interpMode=RCIM_Cubic, leaveTangent=1.4583, time=0.5, value=0.925}
           {arriveTangent=-1.85, interpMode=RCIM_Cubic, leaveTangent=-1.85, time=0.6, value=0.875}
           <at 1 0 RCIM_Cubic>
          ]}}})
     )
     2=&FloatFromCurve(
      FloatCurve=DataInterface() {propertyValues {Curve {keys [<at 0 0.7922>, {arriveTangent=-1.4282, interpMode=RCIM_Cubic, leaveTangent=-1.4282, tangentMode=RCTM_User, time=0.6, value=0.8}, <at 1 0>]}}}
      Scale Curve=2
     )
     3=&FloatFromCurve(
      FloatCurve=DataInterface() {propertyValues {Curve {keys [
          <at 0 1 RCIM_Cubic>
          {arriveTangent=-0.8307, interpMode=RCIM_Cubic, leaveTangent=-0.8307, time=0.4812, value=0.7644}
          {arriveTangent=-4.2795, interpMode=RCIM_Cubic, leaveTangent=-4.2795, tangentMode=RCTM_User, time=0.5844, value=0.5146}
          {arriveTangent=-1.1825, interpMode=RCIM_Cubic, leaveTangent=-1.1825, tangentMode=RCTM_User, time=1, value=0}
         ]}}}
      Scale Curve=2
     )
     4=&FloatFromCurve(
      FloatCurve=DataInterface() {propertyValues {Curve {keys [<at 0 1 RCIM_Cubic>, {arriveTangent=-1.5335, tangentMode=RCTM_User, time=1, value=0}]}}}
      Scale Curve=0.5
     )
     5=&Multiply_Float(
      A=0.2
      B=&FloatFromCurve(FloatCurve=DataInterface() {propertyValues {Curve {keys [
           <at 0 0 RCIM_Cubic>
           {arriveTangent=0.8571, interpMode=RCIM_Cubic, leaveTangent=0.8571, time=0.2, value=1}
           {arriveTangent=-1.25, interpMode=RCIM_Cubic, leaveTangent=-1.25, time=0.7, value=0.6}
           <at 1 0 RCIM_Cubic>
          ]}}})
     )
     6=0
     7=0
     8=0
     9=0
     Index=&SpawnGroup()
     Num=::NewEnumerator8
    )
    Scale RGB=&MergeHelper_VectorSwitch(
     0=&MakeVectorFromLinearColorRGB(Color=&ColorFromCurve(ColorCurve=DataInterface() {propertyValues {
         AlphaCurve {keys []}
         BlueCurve {keys [
           <at 0.0056 10>
           <at 0.1845 10>
           {arriveTangent=-21.7748, interpMode=RCIM_Cubic, leaveTangent=-21.7748, time=0.3426, value=5}
           <at 0.5519 2 RCIM_Cubic>
          ]}
         GreenCurve {keys [
           <at 0.0056 0.8305>
           <at 0.1845 1>
           {arriveTangent=-0.5503, interpMode=RCIM_Cubic, leaveTangent=-0.5503, time=0.3426, value=1.17}
           <at 0.5519 0.7978 RCIM_Cubic>
          ]}
         RedCurve {keys [
           <at 0.0056 0.5>
           <at 0.1845 1.438>
           {arriveTangent=-1.9353, interpMode=RCIM_Cubic, leaveTangent=-1.9353, time=0.3426, value=1.1967}
           <at 0.5519 0.727 RCIM_Cubic>
          ]}
        }}))
     1=&MakeVectorFromLinearColorRGB(Color=<rgba 0.1756 0.1756 1 1>)
     2=&Multiply_VectorByFloat(
      Vector=&MakeVectorFromLinearColorRGB(Color=&ColorFromCurve(ColorCurve=DataInterface() {propertyValues {
          AlphaCurve {keys []}
          BlueCurve {keys [<at 0 2 RCIM_Cubic>,<at 0.3631 2 RCIM_Cubic>,<at 0.6778 2>]}
          GreenCurve {keys [<at 0 0.7896 RCIM_Cubic>, {arriveTangent=-0.0718, interpMode=RCIM_Cubic, leaveTangent=-0.0718, time=0.3631, value=0.56}, <at 0.6778 0.7409>]}
          RedCurve {keys [
            <at 0 0.8485 RCIM_Cubic>
            {arriveTangent=-0.1588, interpMode=RCIM_Cubic, leaveTangent=-0.1588, time=0.3631, value=0.6001}
            <at 0.6778 0.7409>
           ]}
         }}))
      Float=&FloatFromCurve(
       FloatCurve=DataInterface() {propertyValues {Curve {keys [{arriveTangent=-0.0004, interpMode=RCIM_Cubic, leaveTangent=-0.0004, tangentMode=RCTM_User, time=0, value=0.5}, <at 0.5 0.5>]}}}
       Scale Curve=5
      )
     )
     3=&Multiply_VectorByFloat(
      Vector=&MakeVectorFromLinearColorRGB(Color=&ColorFromCurve(ColorCurve=DataInterface() {propertyValues {
          AlphaCurve {keys []}
          BlueCurve {keys [<at 0 3 RCIM_Cubic>, {arriveTangent=-2, interpMode=RCIM_Cubic, leaveTangent=-2, time=0.3549, value=3}, <at 1 1>]}
          GreenCurve {keys [<at 0 0.8648 RCIM_Cubic>, {arriveTangent=-0.417, interpMode=RCIM_Cubic, leaveTangent=-0.417, time=0.3549, value=0.876}, <at 1 0.4478>]}
          RedCurve {keys [<at 0 0.7878 RCIM_Cubic>, {arriveTangent=-0.3478, interpMode=RCIM_Cubic, leaveTangent=-0.3478, time=0.3549, value=0.84}, <at 1 0.44>]}
         }}))
      Float=&FloatFromCurve(FloatCurve=DataInterface() {propertyValues {Curve {keys [<at 0 1>,<at 0.5 1>]}}}, Scale Curve=3)
     )
     4=&VectorFromCurve(VectorCurve=DataInterface() {propertyValues {
        XCurve {keys [<at 0 0.6723>,<at 1 0.4876>]}
        YCurve {keys [{arriveTangent=-0.4825, interpMode=RCIM_Cubic, leaveTangent=-0.4825, tangentMode=RCTM_User, time=0, value=0.35}, <at 1 0.292>]}
        ZCurve {keys [<at 0 1.5889>,<at 1 0.5647>]}
       }})
     5=<1 1 1>
     6=<0 0 0>
     7=<0 0 0>
     8=<0 0 0>
     9=<0 0 0>
     Index=&SpawnGroup()
     Num=::NewEnumerator8
    )
    ScaleA=true
    ScaleRGB=true
   )
   &DynamicMaterialParameters(
    Index 0 Param 1=&MergeHelper_FloatSwitch(
     0=&FloatFromCurve(FloatCurve=DataInterface() {propertyValues {Curve {keys [<at 0 0 RCIM_Cubic>,<at 1 1 RCIM_Cubic>]}}})
     1=&FloatFromCurve(FloatCurve=DataInterface() {propertyValues {Curve {keys [<at 0 0 RCIM_Cubic>,<at 1 0 RCIM_Cubic>]}}})
     2=&FloatFromCurve(FloatCurve=DataInterface() {propertyValues {Curve {keys [<at 0 0 RCIM_Cubic>,<at 1 0>]}}})
     3=&FloatFromCurve(FloatCurve=DataInterface() {propertyValues {Curve {keys [<at 0 0 RCIM_Cubic>,<at 1 1>]}}})
     4=&FloatFromCurve(FloatCurve=DataInterface() {propertyValues {Curve {keys [<at 0 0 RCIM_Cubic>,<at 1 0 RCIM_Cubic>]}}})
     5=&FloatFromCurve(FloatCurve=DataInterface() {propertyValues {Curve {keys [<at 0 0 RCIM_Cubic>,<at 1 0 RCIM_Cubic>]}}})
     6=0
     7=0
     8=0
     9=0
     Index=&SpawnGroup()
     Num=::NewEnumerator8
    )
    Index 0 Param 2=&MergeHelper_FloatSwitch(0=0.6, 1=0.6, 2=0.1, 3=0.1, 4=0.6, 5=0.6, 6=0, 7=0, 8=0, 9=0, Index=&SpawnGroup(), Num=::NewEnumerator8)
    Module.Index0Param0Write=true
    Module.Index0Param1Write=true
    Module.Index0Param2Write=true
    Module.Index0Param3Write=true
   )
   &Drag(Drag=$System.VelocityDrag, Use Linear Drag=true, Use Rotational Drag=true)
   &SolveForcesAndVelocity(Clamp Velocity=false, Limit Acceleration=false)
  ]
  rendererClasses [&NiagaraMeshRendererProperties]
  renderers [
   {rendererClass=&NiagaraMeshRendererProperties, rendererIndex=0}
   {rendererClass=&NiagaraMeshRendererProperties, rendererIndex=1}
   {rendererClass=&NiagaraMeshRendererProperties, rendererIndex=2}
   {rendererClass=&NiagaraMeshRendererProperties, rendererIndex=3}
   {rendererClass=&NiagaraMeshRendererProperties, rendererIndex=4}
   {rendererClass=&NiagaraMeshRendererProperties, rendererIndex=5}
  ]
 }
 Flow {
  bEnabled=false
  emitterUpdateScript [
   &EmitterState(Life Cycle Mode=::NewEnumerator1, Loop Behavior=::NewEnumerator1, Loop Duration=1, UseLoopDelay=true)
   &SpawnBurst_Instantaneous(Use Spawn Probability=false, UseLoopCountLimit=false)
  ]
  particleSpawnScript [
   &InitializeParticle(
    Color=<rgba 1 0.5458 0.5534 1.5>
    Lifetime=1.3065
    Mesh Scale=&Multiply_VectorByFloat(Vector=<2.8716 0.9823 1>, Float=1.4026)
    Mesh Scale Mode=::NewEnumerator4
    Position Offset=&AddVectorToPosition(Position=<316.5416 0 0>, Vector=$System.FXPosition)
    UsePositionOffset=true
    Write Material Random=true
   )
   &InitialMeshOrientation(Mesh Orientation Mode=::NewEnumerator7, Rotation=<0 0 0.5>, Use Rotation Vector=true)
   &InitialMeshOrientation(Mesh Orientation Mode=::NewEnumerator7, Use Rotation Vector=true) {module=InitialMeshOrientation001}
  ]
  particleUpdateScript [
   &ParticleState()
   &ScaleMeshSize(Scale Factor=&Multiply_Vector(
     A=&VectorFromCurve(
      VectorCurve=DataInterface() {propertyValues {XCurve {keys [<at 0 1>,<at 1 1>]}, YCurve {keys [<at 0 1>,<at 1 1>]}, ZCurve {keys [<at 0 1>,<at 1 1>]}}}
      Scale Curve=<1.2 1 1>
     )
     B=$System.SizeScale
    )) {module=ScaleMeshSize001}
   &ScaleColor(
    Scale Alpha=&FloatFromCurve(
     FloatCurve=DataInterface() {propertyValues {Curve {keys [
         <at 0 0 RCIM_Cubic>
         {arriveTangent=-0.1567, interpMode=RCIM_Cubic, leaveTangent=-0.1567, tangentMode=RCTM_User, time=0.1, value=1}
         {arriveTangent=-1.5177, interpMode=RCIM_Cubic, leaveTangent=-1.5177, time=0.3, value=0.9794}
         {arriveTangent=-1.9358, interpMode=RCIM_Cubic, leaveTangent=-1.9358, tangentMode=RCTM_User, time=0.5749, value=0.2793}
         {arriveTangent=-0.6689, interpMode=RCIM_Cubic, leaveTangent=-0.6689, tangentMode=RCTM_User, time=1, value=0}
        ]}}}
     Scale Curve=1.5
    )
    Scale RGB=&MakeVectorFromLinearColorRGB(Color=&ColorFromCurve(ColorCurve=DataInterface() {propertyValues {
        AlphaCurve {keys [<at 0 1 RCIM_Cubic>]}
        BlueCurve {keys [<at 0 2 RCIM_Cubic>,<at 1 2 RCIM_Cubic>]}
        GreenCurve {keys [<at 0 0.7783 RCIM_Cubic>,<at 1 0.4 RCIM_Cubic>]}
        RedCurve {keys [<at 0 0.4696 RCIM_Cubic>,<at 1 0.4 RCIM_Cubic>]}
       }}))
    ScaleA=true
    ScaleRGB=true
   )
   &DynamicMaterialParameters(
    Index 0 Param 1=2
    Index 0 Param 2=100
    Module.Index0Param0Write=true
    Module.Index0Param1Write=true
    Module.Index0Param2Write=true
    Module.Index0Param3Write=true
   )
   &FN_Mesh_Attributes(
    Color=&MakeLinearColorFromVectorAndFloat(
     "Vector (RGB)"=&MakeVectorFromLinearColorRGB(Color=<rgba 0.6522 0.6522 3 5>)
     "Float (Alpha)"=&FloatFromCurve(
      FloatCurve=DataInterface() {propertyValues {Curve {keys [
          <at 0 0 RCIM_Cubic>
          {arriveTangent=0.4858, interpMode=RCIM_Cubic, leaveTangent=0.4858, time=0.2, value=0.9794}
          {arriveTangent=-1.9358, interpMode=RCIM_Cubic, leaveTangent=-1.9358, tangentMode=RCTM_User, time=0.5749, value=0.2793}
          {arriveTangent=-0.6689, interpMode=RCIM_Cubic, leaveTangent=-0.6689, tangentMode=RCTM_User, time=1, value=0}
         ]}}}
      Scale Curve=0.5
     )
    )
    Create Color Attribute=true
    Create Scale Attribute=true
    Scale=&Multiply_Vector(A=<1.2 2.7422 1>, B=&Multiply_VectorByFloat(Vector=<1 1 1>, Float=3.2412))
    Write to Module Namespace=true
   )
   &Drag(Drag=$System.VelocityDrag, Use Linear Drag=true, Use Rotational Drag=true)
   &SolveForcesAndVelocity(Clamp Velocity=false, Limit Acceleration=false)
  ]
  rendererClasses [&NiagaraMeshRendererProperties]
  renderers [
   {rendererClass=&NiagaraMeshRendererProperties, rendererIndex=0}
   {rendererClass=&NiagaraMeshRendererProperties, rendererIndex=1}
  ]
 }
 Minimal {
  bEnabled=false
  emitterUpdateScript [
   &EmitterState(Life Cycle Mode=::NewEnumerator1, Loop Behavior=::NewEnumerator1, UseLoopDelay=false)
   &SpawnBurst_Instantaneous(Use Spawn Probability=false, UseLoopCountLimit=false)
  ]
  particleSpawnScript [
   &InitializeParticle(
    Color=<rgba 0.2122 0.2426 2 0.5>
    Lifetime=0.392
    Position Offset=<167.4118 0 0>
    Sprite Size=<2152.3135 10084.917>
    Sprite Size Mode=::NewEnumerator4
    UsePositionOffset=true
    Write Material Random=true
   )
   &AddVelocity(Velocity=<189.5627 0 0>)
  ]
  particleUpdateScript [
   &ParticleState()
   &SpriteFacingAndAlignment(WriteAlignment=true, WriteFacing=false)
   &Drag(Use Linear Drag=true, Use Rotational Drag=true)
   &SolveForcesAndVelocity(Clamp Velocity=false, Limit Acceleration=false)
   &ScaleColor(
    Scale Alpha=&FloatFromCurve(FloatCurve=DataInterface() {propertyValues {Curve {keys [
         {interpMode=RCIM_Cubic, leaveTangent=9.9505, tangentMode=RCTM_Break, time=0, value=0}
         {interpMode=RCIM_Cubic, tangentMode=RCTM_Break, time=0.3334, value=1}
         <at 1 0>
        ]}}})
    ScaleA=true
    ScaleRGB=true
   )
  ]
  rendererClasses [&NiagaraSpriteRendererProperties]
  renderers [{rendererClass=&NiagaraSpriteRendererProperties}]
 }
 Minimal003 {
  emitterUpdateScript [
   &EmitterState(Life Cycle Mode=::NewEnumerator1, Loop Behavior=::NewEnumerator1, Loop Duration=10, UseLoopDelay=false)
   &SpawnBurst_Instantaneous(Use Spawn Probability=false, UseLoopCountLimit=false)
  ]
  particleSpawnScript [
   &InitializeParticle(
    Lifetime=2
    Mesh Scale=&Multiply_VectorByFloat(Vector=<1 1 1>, Float=1)
    Mesh Scale Mode=::NewEnumerator4
    Position Offset=<-57.5303 0 0>
    UsePositionOffset=true
    Write Material Random=true
   )
   &InitialMeshOrientation(Mesh Orientation Mode=::NewEnumerator7, Rotation=<0.25 0 0>, Use Rotation Vector=true)
  ]
  particleUpdateScript [
   &ParticleState()
   &self:ArrayToInt(Bool Array:NiagaraDataInterfaceArrayBool=$User.ActiveArray) {module=ArrayToInt}
   &DynamicMaterialParameters(
    Index 0 Param 1=$Particles.Mask
    Index 0 Param 2=&FloatFromCurve(FloatCurve=DataInterface() {propertyValues {Curve {keys [
         {interpMode=RCIM_Cubic, leaveTangent=2.4962, tangentMode=RCTM_Break, time=0.25, value=0}
         {arriveTangent=2.6266, interpMode=RCIM_Cubic, tangentMode=RCTM_Break, time=0.8847, value=1}
        ]}}})
    Index 0 Param 4=1
    Module.Index0Float0Write=true
    Module.Index0Float1Write=true
    Module.Index0Float2Write=true
    Module.Index0Float3Write=true
   )
  ]
  rendererClasses [&NiagaraMeshRendererProperties]
  renderers [{rendererClass=&NiagaraMeshRendererProperties, rendererIndex=0}]
 }
 ParticleLight {bEnabled=false, rendererClasses [], renderers []}
 particle {
  bEnabled=false
  emitterUpdateScript [
   &EmitterState(Life Cycle Mode=::NewEnumerator1, Loop Behavior=::NewEnumerator1, Loop Duration=0.3, UseLoopDelay=false)
   &SpawnBurst_Instantaneous(Spawn Count=10, Use Spawn Probability=false, UseLoopCountLimit=false)
   &SpawnRate(SpawnRate=80, Use Spawn Probability=false)
  ]
  particleSpawnScript [
   &InitializeParticle(
    Color=<rgba 3.0058 2.3333 4 1>
    Lifetime Max=0.5
    Lifetime Min=0.3
    Lifetime Mode=::NewEnumerator1
    Sprite Size Max=<0.5 7>
    Sprite Size Min=<0.5 5>
    Sprite Size Mode=::NewEnumerator2
    UsePositionOffset=false
    Write Material Random=true
   )
   &ShapeLocation(
    Box Size=<100 328.8016 1>
    MaskSpawnGroup=false
    Offset=<-119.5825 0 0>
    Offset Mode=::NewEnumerator0
    Override Seed=false
    Shape Primitive=::NewEnumerator2
    Surface Only=false
    UseOwnerScale=true
   )
   &AddVelocity(Velocity=<-3348.5581 0 0>)
   &AddVelocity(Module.Use Falloff Distance=false, Origin Offset=<0 0 100>, Override Seed=false, Velocity Mode=::NewEnumerator1, Velocity Speed=79.3346) {module=AddVelocity001}
  ]
  particleUpdateScript [
   &ParticleState()
   &ScaleSpriteSize(Scale Sprite Size Mode=::NewEnumerator0, Uniform Scale Factor=2)
   &CurlNoiseForce(Bias Noise=false, Noise Strength=3000, Pan Noise=false)
   &Drag(
    Drag=&FloatFromCurve(FloatCurve=DataInterface() {propertyValues {Curve {keys [<at 0 1>,<at 1 0>]}}}, Scale Curve=5)
    Use Linear Drag=true
    Use Rotational Drag=true
   )
   &SolveForcesAndVelocity(Clamp Velocity=false, Limit Acceleration=false)
   &ScaleSpriteSizeBySpeed(Max Scale Factor=<1 2>, Min Scale Factor=<0.5 0.5>)
   &FN_Sprite_Attributes(
    Color=<rgba 0.1604 0.153 0.4209 0.25>
    Create Color Attribute=true
    Create SpriteSize Attribute=true
    SpriteSize=&Multiply_Vector2D(A=$Particles.SpriteSize, B=<30 2>)
    Write to Module Namespace=true
   )
   &DynamicMaterialParameters(
    Index 0 Param 1=2
    Module.Index0Param0Write=true
    Module.Index0Param1Write=true
    Module.Index0Param2Write=true
    Module.Index0Param3Write=true
   )
   /SetVariables(Particles.Position02:NiagaraPosition=&AddVectorToPosition(
     Position=$Particles.Position
     Vector=&MakeVector(X=&Multiply_Float(A=$Emitter.Age, B=1000), Y=0, Z=0)
    )) {isSetParameters=true}
  ]
  rendererClasses [&NiagaraSpriteRendererProperties]
  renderers [
   {rendererClass=&NiagaraSpriteRendererProperties}
   {rendererClass=&NiagaraSpriteRendererProperties, rendererIndex=1}
  ]
 }
}
summary {
 systemName=NS_HeavyAttack_Projectile1
 emitters [
  {bEnabled=false, emitterName=CombineMesh, rendererClasses [&NiagaraMeshRendererProperties]}
  {bEnabled=false, emitterName=Flow, rendererClasses [&NiagaraMeshRendererProperties]}
  {bEnabled=false, emitterName=particle, rendererClasses [&NiagaraSpriteRendererProperties]}
  {emitterName=Minimal003, rendererClasses [&NiagaraMeshRendererProperties]}
  {bEnabled=false, emitterName=ParticleLight, rendererClasses []}
  {bEnabled=false, emitterName=Minimal, rendererClasses [&NiagaraSpriteRendererProperties]}
 ]
 userVariables [
  User.ActiveArray {
   type=NiagaraDataInterfaceArrayBool
   description=""
   defaultValue {
    struct=NiagaraExt_VariableValue_DataInterface
    value {
     dataInterface=&self:NiagaraDataInterfaceArrayBool_0
     dataInterfaceClass=/Script/Niagara.NiagaraDataInterfaceArrayBool
    }
   }
  }
  User.AnimationTimeInput {type=NiagaraFloat, defaultValue=0.7351, description=""}
  User.FXPositionOffset {type=Vector3f, description="", defaultValue {struct=Vector3f, value=<0 0 0>}}
  User.FXvelocityOffset {type=Vector3f, description="", defaultValue {struct=Vector3f, value=<0 0 0>}}
 ]
}
